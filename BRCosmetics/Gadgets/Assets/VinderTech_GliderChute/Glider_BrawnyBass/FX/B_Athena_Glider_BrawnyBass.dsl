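; Brawny Bass glider Blueprint that plays a closing Niagara effect when the parachute is hidden.
; On a visibility change to hidden, if no valid effect is stored, it looks up the active glider particle variant index.
; It attaches the closed-glider system to the player mesh's Root, offset to the parachute mesh's location.
; The stored system receives that index as the VariantColor float parameter.
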(event EventBeginPlay)

(event Collision|EventActorBeginOverlap
  (params (OtherActor :type Actor%20Object%20Reference)))

(event EventTick
  (params (DeltaSeconds :type Float%20%28single-precision%29)))

(event EventOnChuteVisibilityChanged
  (params (bParachuteVisibility :type Boolean))
  (|CallParentFunction bParachuteVisibility)
  (if bParachuteVisibility
    _
    (elif (Utilities|IsValid (Variables|Default|GetAttachedNS) :!class /Script/Engine.KismetSystemLibrary)
      _
      (else
        (bind _index (CosmeticLoadout|GetVariantChannelActiveIndexforSlotActor "(TagName=\"LoadoutSlot.Glider\")" (Variables|FortPlayerParachute|GetPlayerPawn) "(TagName=\"Cosmetics.Variant.Channel.Particle\")" (Utilities|Array|MakeArray "(TagName=\"Cosmetics.Variant.Property.Particle0\")" "(TagName=\"Cosmetics.Variant.Property.Particle1\")")))
        (Variables|Default|SetCurrentVariant _index)
        (bind _returnvalue (Niagara|SpawnSystemAttached "/BRCosmetics/Gadgets/Assets/VinderTech_GliderChute/Glider_BrawnyBass/FX/NS_BrawnyBass_Glider_Closed.NS_BrawnyBass_Glider_Closed" (Variables|/Script/Engine.Character|GetMesh (Variables|FortPlayerParachute|GetPlayerPawn)) "Root" (Math|Transform|InverseTransformLocation (Transformation|GetWorldTransform (Variables|/Script/Engine.Character|GetMesh (Variables|FortPlayerParachute|GetPlayerPawn))) (Transformation|GetWorldLocation (Variables|Parachute|GetParachuteMesh))) "0, 0, 0" "KeepRelativeOffset" true))
        (Variables|Default|SetAttachedNS _returnvalue)
        (Effects|Components|ParticleSystem|SetFloatParameter (Variables|Default|GetAttachedNS) "VariantColor" (Math|Conversions|ToFloat(Integer) (Variables|Default|GetCurrentVariant)))))))

(fn ConstructionScript
  (|CallParentFunction))
