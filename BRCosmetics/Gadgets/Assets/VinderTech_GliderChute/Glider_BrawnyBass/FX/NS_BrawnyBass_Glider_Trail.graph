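# Niagara trail effect for the BrawnyBass glider, built from a single OuterCone mesh-particle emitter.
# A burst spawns particles that outlive their lifetime, and an update-stage skeletal mesh sample keeps them offset and oriented to User.SkeletalMesh.
# Initial rotation differs between gameplay and front end.
# Randomized spawn alpha is scaled by a fixed value in the front end, or by a User.ForwardDot-driven lerp in gameplay.
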NiagaraSystem /BRCosmetics/Gadgets/Assets/VinderTech_GliderChute/Glider_BrawnyBass/FX/NS_BrawnyBass_Glider_Trail
dependencies {usedDataInterfaces [], usedDynamicInputs [], usedModules [&SystemState]}
emitters {OuterCone {
  emitterUpdateScript [
   &EmitterState(Life Cycle Mode=::NewEnumerator1, Loop Behavior=::NewEnumerator1, Loop Duration Mode=::NewEnumerator1, UseLoopDelay=false)
   &SpawnBurst_Instantaneous(Use Spawn Probability=false, UseLoopCountLimit=false)
  ]
  particleSpawnScript [
   &InitializeParticle(
    Color=&MakeLinearColorFromVectorAndFloat(
     "Vector (RGB)"=<1 1 1>
     "Float (Alpha)"=&Multiply_Float(A=&RandomRangeFloat(Minimum=0.75, Maximum=1), B=0.8)
    )
    Lifetime=0.35
    Mesh Scale=<1 1 0.85>
    Mesh Scale Mode=::NewEnumerator4
    Sprite Size Mode=::NewEnumerator3
    UsePositionOffset=false
    Write Material Random=true
   )
   &InitialMeshOrientation(
    Orientation Axis=<0 -1 0>
    Orientation Coordinate Space=::Local
    Rotation=&FN_FrontEndVector3Modifier(Gameplay=<0 -0.05 0>, Front End=<0.35 0 0>)
    Use Orientation Vector=true
    Use Rotation Vector=true
   )
  ]
  particleUpdateScript [
   &ParticleState(Kill Particles When Lifetime Has Elapsed=false, Let Infinitely Lived Particles Die When Emitter Deactivates=true)
   &SkeletalMeshLocation(
    Apply Sampled Orientation to Mesh Particles=true
    Inherit Velocity=false
    MaskSpawnGroup=false
    Mesh Sampling Type=::NewEnumerator5
    Sampled Position Offset=<0 75 35>
    Sampled Position Offset Coordinate Space:ENiagaraMeshTransforms=::NewEnumerator3
    Skeletal Mesh=$User.SkeletalMesh
    Socket Sampling Mode=::NewEnumerator5
   )
   &ScaleColor(
    Scale Alpha=&MakeCustomFloatFromBool(Bool=$User.bIsFrontEnd, True Float=0.3, False Float=&Lerp_Float(A=0.1, B=1, Alpha=$User.ForwardDot))
    ScaleA=true
    ScaleRGB=true
   )
  ]
  rendererClasses [&NiagaraMeshRendererProperties]
  renderers [{rendererClass=&NiagaraMeshRendererProperties, rendererIndex=0}]
 }}
summary {
 systemName=NS_BrawnyBass_Glider_Trail
 emitters [{emitterName=OuterCone, rendererClasses [&NiagaraMeshRendererProperties]}]
 userVariables [
  User.bIsFrontEnd {type=NiagaraBool}
  User.bIsFrontEndPreview {type=NiagaraBool}
  User.ForwardDot {type=NiagaraFloat, defaultValue=0, description=""}
  User.SkeletalMesh {
   type=NiagaraDataInterfaceSkeletalMesh
   defaultValue {
    struct=NiagaraExt_VariableValue_DataInterface
    value {dataInterface=&self:NiagaraDataInterfaceSkeletalMesh_0}
   }
  }
 ]
}
